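# Physics asset for the Night Bloom two-handed pickaxe mesh, providing collision bodies for the handle and a jointed dangling charm.
# A long capsule covers the handle, a tiny one marks the MSTR_charm root, and three capsules form the dyn_charm segments.
# Constraints chain each segment to the one before, starting at MSTR_charm, with zero twist limits and uneven swing limits that favor movement on the second swing axis.
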PhysicsAsset /BRCosmetics/Weapons/FORT_Melee/Pickaxe_NightBloom/Meshes/NightBloom_Axe_FE_Physics
bodies {
 MSTR_charm [{shapeType=Capsule, length=0.505, radius=0.505, shapeName=MSTR_charm_capsule, rotation=<rot 0 -90 0>}]
 Melee_TwoHanded_Handle [{
   shapeType=Capsule
   length=61.1737
   radius=2.7122
   shapeName=Melee_TwoHanded_Handle_capsule
   center=<-0.3944 0.7715 14.9326>
   rotation=<rot 1.6498 -89.9999 -0.007>
  }]
 dyn_charm_01 [{shapeType=Capsule, length=3, radius=2, shapeName=dyn_charm_01_capsule, center=<2 0 0>, rotation=<rot -90 0 0>}]
 dyn_charm_02 [{shapeType=Capsule, length=3, radius=2, shapeName=dyn_charm_02_capsule, center=<2 0 0>, rotation=<rot -90 0 0>}]
 dyn_charm_03 [{shapeType=Capsule, length=3, radius=2, shapeName=dyn_charm_02_capsule, center=<2 0 0>, rotation=<rot -90 0 0>}]
}
constraints [
 {bone1Name=dyn_charm_01, bone2Name=MSTR_charm, swing1LimitDegrees=1, swing2LimitDegrees=110, twistLimitDegrees=0}
 {bone1Name=dyn_charm_02, bone2Name=dyn_charm_01, swing1LimitDegrees=45, swing2LimitDegrees=90, twistLimitDegrees=0}
 {bone1Name=dyn_charm_03, bone2Name=dyn_charm_02, swing1LimitDegrees=20, swing2LimitDegrees=180, twistLimitDegrees=0}
]
